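# Translucent unlit deferred decal for a blast mark, filed under the Swarm Boss spawn effects, with a glowing energy center over a dark base color.
# Radial UVs drive a panning, noise-distorted energy wave, tinted by GlowColor and masked by a scorch-hole texture plus tiling dust.
# A ground-char texture plus the center glow shapes a mask, eroded by inky smoke.
# Decal color G scales glow, B scales that mask, and A offsets erosion.
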Material /TidalCrane_Swarm_Boss/Assets/FX/EventVariant/Spawn/MD_Decal_Blast
graph {
 properties {
  bAutomaticallySetUsageInEditor=false
  blendMode=BLEND_Translucent
  materialDomain=MD_DeferredDecal
  pixelDepthOffsetMode=PDOM_Legacy
  shadingModel=MSM_Unlit
  usedShadingModels=Unlit
 }
 nodes [
  Add()
  n1 = TextureObjectParameter() {
   group=EnergyWave
   parameterName=EnergyWave_Texture
   samplerType=SAMPLERTYPE_LinearGrayscale
   sortPriority=0
   texture=/Game/Effects/Fort_Effects/Emotes/UltraLight/Texture/T_Plasma_A_Greyscale
  }
  n2 = TextureObjectParameter() {
   group=EnergyNoise
   parameterName=EnergyNoise_Texture
   samplerType=SAMPLERTYPE_LinearGrayscale
   sortPriority=0
   texture=/Game/Effects/Fort_Effects/Emotes/UltraLight/Texture/T_Cloud_Greyscale
  }
  n3 = VectorParameter() {group=EnergyNoise, parameterName=EnergyNoise_UVControl, channelNames {a=VPan, b=UPan, g=VScale, r=UScale}, defaultValue=<rgba 1 1 0 0.2>}
  n4 = MaterialFunctionCall() {materialFunction=&VectorToRadialValue}
  n5 = VectorParameter() {
   group=EnergyNoise
   parameterName=EnergyNoise_NoiseControl
   sortPriority=60
   channelNames {a=VStrength, b=UStrength, g=VOffset, r=UOffset}
   defaultValue=<rgba -0.5 -0.5 0.1 0.12>
  }
  n6 = MaterialFunctionCall!(
   NoiseValue=MaterialFunctionCall(Texture=@n2, UVControl_RGB=@n3, UVControl_A=@n3, "[Override]UVInput"=@n4.Radial Coordinates) {materialFunction=&FXCommon_TextureSample}.RG
   Control_RGB=@n5
   Control_A=@n5
  ) {materialFunction=/Game/Effects/Niagara/MasterMaterials/Functions/Common/FXCommon_NoiseControl}
  n7 = VectorParameter() {group=EnergyWave, parameterName=EnergyWave_UVControl, channelNames {a=VPan, b=UPan, g=VScale, r=UScale}, defaultValue=<rgba 1 2 0 -0.354>}
  n8 = VectorParameter() {parameterName=ER, defaultValue=<rgba 15 1000 0.5 0.5>}
  n9 = OneMinus(None=MaterialFunctionCall(Float2=@n4.Radial Coordinates) {materialFunction=&BreakOutFloat2Components}.G)
  n10 = ScalarParameter() {defaultValue=100, parameterName=sD}
  n11 = LinearInterpolate(
   B=MaterialFunctionCall(
    Texture=MaterialFunctionCall(Texture=@n1, AdditionalUVOffset=@n6.NoiseValue, UVControl_RGB=@n7, UVControl_A=@n7, "[Override]UVInput"=@n4.Radial Coordinates) {materialFunction=&FXCommon_TextureSample}.R
    RGB=@n8.RGB
    A=@n8.RGB
   ) {materialFunction=&MF_FX_Refine}.Result
   Alpha=OneMinus(None=Power(Base=@n9, Exp=@n10))
  ) {constA=1}
  n12 = VectorParameter() {parameterName=GlowColor, defaultValue=<rgba 5 28.75 100 4>}
  n13 = DecalColor()
  n14 = Multiply(A=@n12.RGB, B=@n13.G)
  n15 = TextureObjectParameter() {
   group=CenterGlow
   parameterName=CenterGlow_Texture
   sortPriority=0
   texture=/Game/Effects/Fort_Effects/Gameplay/SmokeDustExplosions/Textures/T_Decal_Scorch_Hole
  }
  n16 = VectorParameter() {group=CenterGlow, parameterName=CenterGlow_UVControl, channelNames {a=VPan, b=UPan, g=VScale, r=UScale}, defaultValue=<rgba 2 2 0 0>}
  n17 = Power(
   Base=MaterialFunctionCall!(
    Texture=@n15
    AdditionalUVOffset
    UVControl_RGB=@n16
    UVControl_A=@n16
    "[Override]TimeInput"
    "[Override]ScaleUVByCenter"=StaticBool() {value=true}
    "[Override]UVInput"
   ) {materialFunction=/Game/Effects/Niagara/MasterMaterials/Functions/Common/FXCommon_TextureSample_V2}.R
   Exp=@n12.A
  )
  n18 = TextureObjectParameter() {
   group=EnergyTexture_Sand
   parameterName=EnergyTexture_Sand_Texture
   sortPriority=0
   texture=/Game/Global/Textures/Default/PointSplatter/T_Tiling_Dust
  }
  n19 = VectorParameter() {
   group=EnergyTexture_Sand
   parameterName=EnergyTexture_Sand_UVControl
   channelNames {a=VPan, b=UPan, g=VScale, r=UScale}
   defaultValue=<rgba 1 1 0 0>
  }
  n20 = MaterialFunctionCall(Texture=@n18, UVControl_RGB=@n19, UVControl_A=@n19) {materialFunction=&FXCommon_TextureSample}
  n21 = ScalarParameter() {defaultValue=10, parameterName=test}
  n22 = Saturate(None=Add(A=@n17, B=Multiply(A=@n20.R, B=@n21)))
  Add(
   A=Multiply(A=Multiply(A=@n11, B=@n14), B=Multiply(A=@n22, B=@n17))
   B=VectorParameter() {parameterName=BaseColor, defaultValue=<rgba 0.012 0.028 0.06 1>}.RGB
  )
  n23 = Saturate(None=Multiply(A=@n17, B=ScalarParameter() {defaultValue=10, parameterName=Test}))
  n24 = TextureSampleParameter2D() {
   parameterName=DecalTexture
   samplerType=SAMPLERTYPE_LinearColor
   texture=/Game/Effects/Fort_Effects/Gameplay/SmokeDustExplosions/Textures/T_GroundChar_01
  }
  n25 = TextureObjectParameter() {group=Erosion, parameterName=Erosion_Texture, sortPriority=0, texture=&T_Inky_Smoke_Tile}
  n26 = VectorParameter() {group=Erosion, parameterName=Erosion_UVControl, channelNames {a=VPan, b=UPan, g=VScale, r=UScale}, defaultValue=<rgba 1.5 1.5 0 0>}
  n27 = Multiply(
   A=Add(
    A=MaterialFunctionCall(Texture=@n25, UVControl_RGB=@n26, UVControl_A=@n26) {materialFunction=&FXCommon_TextureSample}.R
    B=@n13.A
   )
   B=ScalarParameter() {defaultValue=3, parameterName=ErosionHard}
  )
  Multiply(A=@n13.B, B=Multiply(A=Add(A=@n23, B=@n24.R), B=Saturate(None=@n27)))
  ScalarParameter() {defaultValue=0.7, parameterName=Metallic}
  ScalarParameter() {parameterName=Specular}
  ScalarParameter() {defaultValue=1, parameterName=Roughness}
 ]
}
