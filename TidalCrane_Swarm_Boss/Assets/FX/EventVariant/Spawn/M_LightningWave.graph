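# Additive, unlit, two-sided Niagara sprite material for a lightning wave effect, filed under the Swarm Boss spawn FX.
# Two panning plasma samples share power-curved UVs, one distorted by Perlin noise, then multiply and pass through a refine function.
# Dynamic parameters supply the panning time and blend the energy toward zero before particle color tints it.
# U and V edge masks multiply with particle alpha to shape the sprite.
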Material /TidalCrane_Swarm_Boss/Assets/FX/EventVariant/Spawn/M_LightningWave
graph {
 properties {
  bUsedWithNiagaraSprites=true
  blendMode=BLEND_Additive
  pixelDepthOffsetMode=PDOM_Legacy
  shadingModel=MSM_Unlit
  twoSided=true
  usedShadingModels=Unlit
 }
 nodes [
  n1 = ParticleColor()
  n2 = MaterialFunctionCall(Float2=TextureCoordinate()) {materialFunction=&BreakOutFloat2Components}
  n3 = VectorParameter() {group=UMask, parameterName=UMask_EdgeMask_Control, channelNames {a="", b=Power, g=Multiply, r=Offset}, defaultValue=<rgba 0 2 3 1>}
  n4 = VectorParameter() {group=VMask, parameterName=VMask_EdgeMask_InvertControl, channelNames {a="", b=Power, g=Multiply, r=Offset}, defaultValue=<rgba 0 3 3 1>}
  n5 = VectorParameter() {group=VMask, parameterName=VMask_EdgeMask_Control, channelNames {a="", b=Power, g=Multiply, r=Offset}, defaultValue=<rgba 0 1 3 1>}
  Multiply(
   A=@n1.A
   B=Multiply(
    A=MaterialFunctionCall!(UorV=@n2.R, EdgeControl=@n3) {materialFunction=/Game/Effects/Niagara/MasterMaterials/Functions/Common/FXCommon_EdgeMask_Simple}.Mask
    B=MaterialFunctionCall!(UorV=@n2.G, Invert_EdgeControl=@n4, NoInvert_EdgeControl=@n5) {materialFunction=/Game/Effects/Niagara/MasterMaterials/Functions/Common/FXCommon_EdgeMask_Ex}.Mask
   )
  )
  n6 = TextureObjectParameter() {
   group=Energy2
   parameterName=Energy2_Texture
   samplerType=SAMPLERTYPE_LinearGrayscale
   sortPriority=0
   texture=/Game/Effects/Fort_Effects/Emotes/UltraLight/Texture/T_Plasma_A_Greyscale
  }
  n7 = TextureObjectParameter() {
   group=Noise
   parameterName=Noise_Texture
   samplerType=SAMPLERTYPE_LinearGrayscale
   sortPriority=0
   texture=&T_SimplePerlin_Greyscale
  }
  n8 = VectorParameter() {group=Noise, parameterName=Noise_UVControl, channelNames {a=VPan, b=UPan, g=VScale, r=UScale}, defaultValue=<rgba 0.3 1.2 0 0.5>}
  n9 = DynamicParameter() {paramNames [Blend,time,Param3,Param4]}
  n10 = MaterialFunctionCall(Float2=TextureCoordinate()) {materialFunction=&BreakOutFloat2Components}
  n11 = Power(Base=OneMinus(None=@n10.G), Exp=ScalarParameter() {defaultValue=0.5, parameterName=FW})
  n12 = AppendVector(A=@n10.R, B=Power(Base=OneMinus(None=@n11), Exp=ScalarParameter() {defaultValue=1.3, parameterName=fwae}))
  n13 = VectorParameter() {
   group=Noise
   parameterName=Noise_NoiseControl
   sortPriority=60
   channelNames {a=VStrength, b=UStrength, g=VOffset, r=UOffset}
   defaultValue=<rgba -0.5 -0.5 0.3 0.15>
  }
  n14 = VectorParameter() {group=Energy2, parameterName=Energy2_UVControl, channelNames {a=VPan, b=UPan, g=VScale, r=UScale}, defaultValue=<rgba 0.5 1.6 0 0.212>}
  n15 = MaterialFunctionCall(
   Texture=@n6
   AdditionalUVOffset=MaterialFunctionCall!(
    NoiseValue=MaterialFunctionCall(Texture=@n7, UVControl_RGB=@n8, UVControl_A=@n8, "[Override]TimeInput"=@n9.time, "[Override]UVInput"=@n12) {materialFunction=&FXCommon_TextureSample}.RG
    Control_RGB=@n13
    Control_A=@n13
   ) {materialFunction=/Game/Effects/Niagara/MasterMaterials/Functions/Common/FXCommon_NoiseControl}.NoiseValue
   UVControl_RGB=@n14
   UVControl_A=@n14
   "[Override]TimeInput"=@n9.time
   "[Override]UVInput"=@n12
  ) {materialFunction=&FXCommon_TextureSample}
  n16 = ScalarParameter() {defaultValue=0.8, parameterName=awdawda}
  n17 = TextureObjectParameter() {
   group=Energy
   parameterName=Energy_Texture
   samplerType=SAMPLERTYPE_LinearGrayscale
   sortPriority=0
   texture=&T_Plasma_A_Greyscale
  }
  n18 = VectorParameter() {group=Energy, parameterName=Energy_UVControl, channelNames {a=VPan, b=UPan, g=VScale, r=UScale}, defaultValue=<rgba 0.8 1 0 0.6>}
  n19 = VectorParameter() {parameterName=EnergyRefine, defaultValue=<rgba 2 15 7.5 0.5>}
  n20 = MaterialFunctionCall(
   Texture=Multiply(
    A=Power(Base=@n15.R, Exp=@n16)
    B=MaterialFunctionCall(Texture=@n17, UVControl_RGB=@n18, UVControl_A=@n18, "[Override]TimeInput"=@n9.time, "[Override]UVInput"=@n12) {materialFunction=&FXCommon_TextureSample}.R
   )
   RGB=@n19.RGB
   A=@n19.RGB
  ) {materialFunction=&MF_FX_Refine}
  n21 = Constant()
  Multiply(A=@n1.RGB, B=Add(A=LinearInterpolate(A=@n20.Result, B=@n21, Alpha=@n9.Blend)))
  Multiply()
  Power(Base=@n20.Result, Exp=ScalarParameter() {defaultValue=0.1, parameterName=aef32})
  VectorParameter() {parameterName=Test, defaultValue=<rgba 4 9.3333 20 1>}
 ]
}
